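# Physics asset for the hanging pant loops on the medium male LettuceCat cosmetic body, with capsule bodies for the pelvis, thighs, calves and loop bones.
# Each side has front and back bone chains that hang from master bodies, which pivot freely on the thigh.
# Chain links mostly use limited swing and tight twist.
# The last front and back links join through a free-rotating constraint, closing each chain pair into a loop.
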PhysicsAsset /BRCosmetics/Characters/Player/Male/Medium/Bodies/M_MED_LettuceCat/Meshes/M_MED_LettuceCat_PantLoop_Physics
bodies {
 calf_l [{shapeType=Capsule, length=39.25, radius=7.75, shapeName=calf_l_capsule, center=<-10.8305 -1.1241 0.943>, rotation=<rot 90 10.0248 10.0252>}]
 calf_r [{shapeType=Capsule, length=37.0148, radius=7.4852, shapeName=calf_r_capsule, center=<10.6744 1.6477 -0.6084>, rotation=<rot -90 180 179.9989>}]
 dyn_PantLoop_L_MSTR [{shapeType=Capsule, length=0.505, radius=0.505, shapeName=dyn_PantLoop_L_MSTR_capsule, rotation=<rot 0 -90 0>}]
 dyn_PantLoop_L_Top [{
   shapeType=Capsule
   length=36.0005
   radius=3
   shapeName=dyn_pantLoop_L_bck_01_capsule
   center=<7.6069 4.7994 4.4063>
   rotation=<rot 0 0 112.2573>
  }]
 dyn_PantLoop_L_attach [{shapeType=Capsule, length=9.0035, radius=1.9965, shapeName=dyn_pantLoop_L_bck_01_capsule, center=<7.6069 0 5>, rotation=<rot 90 0 0>}]
 dyn_pantLoop_L_base [{
   shapeType=Capsule
   length=10.4522
   radius=1.2251
   shapeName=dyn_pantLoop_L_base_capsule
   center=<-1.3714 1.3421 -0.5818>
   rotation=<rot 36.7473 -90 30.3534>
  }]
 dyn_pantLoop_L_bck_01 [{shapeType=Capsule, length=9.0035, radius=1.9965, shapeName=dyn_pantLoop_L_bck_01_capsule, center=<7.6069 0 5>, rotation=<rot 90 0 0>}]
 dyn_pantLoop_L_bck_02 [{shapeType=Capsule, length=7.5944, radius=3.4056, shapeName=dyn_pantLoop_L_bck_02_capsule, center=<7.6069 0 5>, rotation=<rot 90 0 0>}]
 dyn_pantLoop_L_bck_03 [{shapeType=Capsule, length=9.3101, radius=3.8473, shapeName=dyn_pantLoop_L_bck_03_capsule, center=<7.6069 0 5>, rotation=<rot 90 0 0>}]
 dyn_pantLoop_L_bck_04 [{shapeType=Capsule, length=16.2825, radius=3.6949, shapeName=dyn_pantLoop_L_bck_03_capsule, center=<7.6069 0 5>, rotation=<rot 90 0 0>}]
 dyn_pantLoop_L_fr_01 [{shapeType=Capsule, length=12.9276, radius=4.1882, shapeName=dyn_pantLoop_L_fr_01_capsule, center=<7.6069 0 5>, rotation=<rot 90 0 0>}]
 dyn_pantLoop_L_fr_02 [{shapeType=Capsule, length=12.9276, radius=4.1882, shapeName=dyn_pantLoop_L_fr_02_capsule, center=<7.6069 0 5>, rotation=<rot 90 0 0>}]
 dyn_pantLoop_L_fr_03 [{shapeType=Capsule, length=13.3567, radius=5.9165, shapeName=dyn_pantLoop_L_fr_03_capsule, center=<7.6069 0 5>, rotation=<rot 90 0 0>}]
 dyn_pantLoop_L_fr_04 [{shapeType=Capsule, length=16.2825, radius=3.6949, shapeName=dyn_pantLoop_L_fr_03_capsule, center=<7.6069 0 5>, rotation=<rot 90 0 0>}]
 dyn_pantLoop_L_fr_MSTR [{shapeType=Capsule, length=0.505, radius=0.505, shapeName=dyn_pantLoop_L_fr_MSTR_capsule, rotation=<rot 0 -90 0>}]
 dyn_pantLoop_L_fr_Top [{shapeType=Capsule, length=11.4914, radius=5.6244, shapeName=dyn_pantLoop_L_fr_01_capsule, center=<7.6069 0 5>, rotation=<rot 90 0 0>}]
 dyn_pantLoop_R_MSTR [{shapeType=Capsule, length=0.505, radius=0.505, shapeName=dyn_pantLoop_R_MSTR_capsule, rotation=<rot 0 -90 0>}]
 dyn_pantLoop_R_attach [{shapeType=Capsule, length=5.997, radius=5.003, shapeName=dyn_pantLoop_R_bck_01_capsule, center=<7.6069 0 4.6619>, rotation=<rot 90 0 0>}]
 dyn_pantLoop_R_base [{shapeType=Capsule, length=0.505, radius=0.505, shapeName=dyn_pantLoop_R_base_capsule, rotation=<rot 0 -90 0>}]
 dyn_pantLoop_R_bck_01 [{shapeType=Capsule, length=5.997, radius=5.003, shapeName=dyn_pantLoop_R_bck_01_capsule, center=<7.6069 0 5>, rotation=<rot 90 0 0>}]
 dyn_pantLoop_R_bck_02 [{shapeType=Capsule, length=5.997, radius=5.003, shapeName=dyn_pantLoop_R_bck_02_capsule, center=<7.6069 0 5>, rotation=<rot 90 0 0>}]
 dyn_pantLoop_R_bck_03 [{shapeType=Capsule, length=9.3101, radius=3.8473, shapeName=dyn_pantLoop_R_bck_03_capsule, center=<7.6069 0 5>, rotation=<rot 90 0 0>}]
 dyn_pantLoop_R_bck_04 [{shapeType=Capsule, length=16.2825, radius=3.6949, shapeName=dyn_pantLoop_R_bck_03_capsule, center=<7.6069 0 5>, rotation=<rot 90 0 0>}]
 dyn_pantLoop_R_fr_01 [{shapeType=Capsule, length=11.9888, radius=5.1271, shapeName=dyn_pantLoop_R_fr_01_capsule, center=<7.6069 0 3.8298>, rotation=<rot 90 0 0>}]
 dyn_pantLoop_R_fr_02 [{shapeType=Capsule, length=13.101, radius=4.0148, shapeName=dyn_pantLoop_R_fr_02_capsule, center=<7.6069 0 5>, rotation=<rot 90 0 0>}]
 dyn_pantLoop_R_fr_03 [{shapeType=Capsule, length=13.3567, radius=5.9165, shapeName=dyn_pantLoop_R_fr_03_capsule, center=<7.6069 0 5>, rotation=<rot 90 0 0>}]
 dyn_pantLoop_R_fr_04 [{shapeType=Capsule, length=16.2825, radius=3.6949, shapeName=dyn_pantLoop_R_fr_03_capsule, center=<7.6069 0 5>, rotation=<rot 90 0 0>}]
 dyn_pantLoop_R_fr_MSTR [{shapeType=Capsule, length=0.505, radius=0.505, shapeName=dyn_pantLoop_R_fr_MSTR_capsule, rotation=<rot 0 -90 0>}]
 dyn_pantLoop_R_fr_Top [{shapeType=Capsule, length=9.5759, radius=7.5399, shapeName=dyn_pantLoop_R_fr_01_capsule, center=<7.6069 0 4.7109>, rotation=<rot 90 0 0>}]
 pelvis [{shapeType=Capsule, length=14.75, radius=12, shapeName=pelvis_capsule, center=<-3.4994 0 0>}]
 thigh_l [{
   shapeType=Capsule
   length=32.6383
   radius=11.3617
   shapeName=thigh_l_capsule
   center=<-18.978 -2.2878 1.0261>
   rotation=<rot 85 -0.0002 0.0001>
  }]
 thigh_r [
  {shapeType=Capsule, length=34.75, radius=9.75, shapeName=thigh_r_capsule, center=<18.978 -0.0002 -1.0261>, rotation=<rot -85 180 179.999>}
  {shapeType=Capsule, length=3, radius=7.5, shapeName=thigh_r_capsule, center=<20.9205 0 4.794>, rotation=<rot 90 179.8417 179.8419>}
 ]
}
constraints [
 {
  bone1Name=dyn_pantLoop_L_base
  bone2Name=thigh_l
  swing1LimitDegrees=45
  swing1Motion=Locked
  swing2LimitDegrees=40
  twistLimitDegrees=45
  twistMotion=Locked
 }
 {
  bone1Name=calf_r
  bone2Name=thigh_r
  swing1LimitDegrees=45
  swing1Motion=Free
  swing2LimitDegrees=45
  swing2Motion=Free
  twistLimitDegrees=45
  twistMotion=Free
 }
 {
  bone1Name=calf_l
  bone2Name=thigh_l
  swing1LimitDegrees=45
  swing1Motion=Free
  swing2LimitDegrees=45
  swing2Motion=Free
  twistLimitDegrees=45
  twistMotion=Free
 }
 {
  bone1Name=dyn_pantLoop_R_base
  bone2Name=thigh_r
  swing1LimitDegrees=45
  swing1Motion=Locked
  swing2LimitDegrees=45
  twistLimitDegrees=45
  twistMotion=Locked
 }
 {bone1Name=dyn_pantLoop_R_fr_01, bone2Name=dyn_pantLoop_R_fr_Top, swing1LimitDegrees=35, swing2LimitDegrees=35, twistLimitDegrees=3}
 {bone1Name=dyn_pantLoop_R_fr_03, bone2Name=dyn_pantLoop_R_fr_02, swing1LimitDegrees=35, swing2LimitDegrees=35, twistLimitDegrees=3}
 {bone1Name=dyn_pantLoop_R_fr_02, bone2Name=dyn_pantLoop_R_fr_01, swing1LimitDegrees=35, swing2LimitDegrees=35, twistLimitDegrees=3}
 {bone1Name=dyn_pantLoop_L_bck_01, bone2Name=dyn_PantLoop_L_attach, swing1LimitDegrees=35, swing2LimitDegrees=35, twistLimitDegrees=3}
 {bone1Name=dyn_pantLoop_L_bck_02, bone2Name=dyn_pantLoop_L_bck_01, swing1LimitDegrees=35, swing2LimitDegrees=35, twistLimitDegrees=3}
 {bone1Name=dyn_pantLoop_L_bck_03, bone2Name=dyn_pantLoop_L_bck_02, swing1LimitDegrees=35, swing2LimitDegrees=35, twistLimitDegrees=3}
 {bone1Name=dyn_pantLoop_R_bck_01, bone2Name=dyn_pantLoop_R_attach, swing1LimitDegrees=35, swing2LimitDegrees=35, twistLimitDegrees=3}
 {bone1Name=dyn_pantLoop_R_bck_02, bone2Name=dyn_pantLoop_R_bck_01, swing1LimitDegrees=35, swing2LimitDegrees=35, twistLimitDegrees=3}
 {bone1Name=dyn_pantLoop_R_bck_03, bone2Name=dyn_pantLoop_R_bck_02, swing1LimitDegrees=35, swing2LimitDegrees=35, twistLimitDegrees=3}
 {bone1Name=dyn_pantLoop_L_fr_01, bone2Name=dyn_pantLoop_L_fr_Top, swing1LimitDegrees=35, swing2LimitDegrees=35, twistLimitDegrees=3}
 {bone1Name=dyn_pantLoop_L_fr_02, bone2Name=dyn_pantLoop_L_fr_01, swing1LimitDegrees=35, swing2LimitDegrees=35, twistLimitDegrees=3}
 {bone1Name=dyn_pantLoop_L_fr_03, bone2Name=dyn_pantLoop_L_fr_02, swing1LimitDegrees=35, swing2LimitDegrees=35, twistLimitDegrees=3}
 {
  bone1Name=dyn_PantLoop_L_MSTR
  bone2Name=thigh_l
  swing1LimitDegrees=45
  swing1Motion=Free
  swing2LimitDegrees=45
  swing2Motion=Free
  twistLimitDegrees=45
  twistMotion=Free
 }
 {
  bone1Name=dyn_pantLoop_L_fr_MSTR
  bone2Name=thigh_l
  swing1LimitDegrees=45
  swing1Motion=Free
  swing2LimitDegrees=45
  swing2Motion=Free
  twistLimitDegrees=45
  twistMotion=Free
 }
 {
  bone1Name=dyn_pantLoop_R_MSTR
  bone2Name=thigh_r
  swing1LimitDegrees=45
  swing1Motion=Free
  swing2LimitDegrees=45
  swing2Motion=Free
  twistLimitDegrees=45
  twistMotion=Free
 }
 {
  bone1Name=dyn_pantLoop_R_fr_MSTR
  bone2Name=thigh_r
  swing1LimitDegrees=45
  swing1Motion=Free
  swing2LimitDegrees=45
  swing2Motion=Free
  twistLimitDegrees=45
  twistMotion=Free
 }
 {bone1Name=dyn_PantLoop_L_attach, bone2Name=dyn_PantLoop_L_Top, swing1LimitDegrees=35, swing2LimitDegrees=35, twistLimitDegrees=3}
 {bone1Name=dyn_pantLoop_L_fr_Top, bone2Name=dyn_pantLoop_L_fr_MSTR, swing1LimitDegrees=35, swing2LimitDegrees=35, twistLimitDegrees=3}
 {
  bone1Name=dyn_pantLoop_R_attach
  bone2Name=dyn_pantLoop_R_MSTR
  swing1LimitDegrees=35
  swing2LimitDegrees=35
  twistLimitDegrees=0
  twistMotion=Locked
 }
 {bone1Name=dyn_pantLoop_R_fr_Top, bone2Name=dyn_pantLoop_R_fr_MSTR, swing1LimitDegrees=35, swing2LimitDegrees=35, twistLimitDegrees=3}
 {bone1Name=dyn_pantLoop_L_bck_04, bone2Name=dyn_pantLoop_L_bck_03, swing1LimitDegrees=35, swing2LimitDegrees=35, twistLimitDegrees=3}
 {bone1Name=dyn_pantLoop_L_fr_04, bone2Name=dyn_pantLoop_L_fr_03, swing1LimitDegrees=35, swing2LimitDegrees=35, twistLimitDegrees=3}
 {bone1Name=dyn_pantLoop_R_bck_04, bone2Name=dyn_pantLoop_R_bck_03, swing1LimitDegrees=35, swing2LimitDegrees=35, twistLimitDegrees=3}
 {bone1Name=dyn_pantLoop_R_fr_04, bone2Name=dyn_pantLoop_R_fr_03, swing1LimitDegrees=35, swing2LimitDegrees=35, twistLimitDegrees=3}
 {
  bone1Name=dyn_pantLoop_R_bck_04
  bone2Name=dyn_pantLoop_R_fr_04
  swing1LimitDegrees=45
  swing1Motion=Free
  swing2LimitDegrees=45
  swing2Motion=Free
  twistLimitDegrees=45
  twistMotion=Free
 }
 {
  bone1Name=dyn_pantLoop_L_bck_04
  bone2Name=dyn_pantLoop_L_fr_04
  swing1LimitDegrees=45
  swing1Motion=Free
  swing2LimitDegrees=45
  swing2Motion=Free
  twistLimitDegrees=45
  twistMotion=Free
 }
 {bone1Name=dyn_PantLoop_L_Top, bone2Name=dyn_PantLoop_L_MSTR, swing1LimitDegrees=15, swing2LimitDegrees=60, twistLimitDegrees=15}
]
