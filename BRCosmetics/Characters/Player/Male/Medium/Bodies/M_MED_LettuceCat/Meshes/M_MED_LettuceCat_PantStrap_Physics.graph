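# Physics asset for the LettuceCat male medium body, defining capsule bodies for its dangling pant straps, jacket strap and chest zipper.
# Capsules on the spine, thighs and calves form the body volume, linked by free-rotating constraints within each pair.
# Strap chains constrain to the thighs, including the jacket strap on the left, while the zipper hangs from spine_03 through a master bone.
# Twist is locked on jacket and right straps, narrowly limited on left straps and zipper.
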PhysicsAsset /BRCosmetics/Characters/Player/Male/Medium/Bodies/M_MED_LettuceCat/Meshes/M_MED_LettuceCat_PantStrap_Physics
bodies {
 calf_l [{shapeType=Capsule, length=20, radius=8.75, shapeName=calf_l_capsule, center=<-14.6791 -1.6551 0>, rotation=<rot 90 90 90.0002>}]
 calf_r [{shapeType=Capsule, length=20, radius=8.75, shapeName=calf_r_capsule, center=<14.9548 1.7962 0>, rotation=<rot -90 179.9802 -179.9807>}]
 dyn_chestZipper [{
   shapeType=Capsule
   length=3.0193
   radius=1.3815
   shapeName=dyn_chestZipper_capsule
   center=<2.0696 -0.0009 1>
   rotation=<rot 0 -89.595 -89.057>
  }]
 dyn_chestZipper_MSTR [{shapeType=Capsule, length=0.505, radius=0.505, shapeName=dyn_chestZipper_MSTR_capsule, rotation=<rot 0 -90 0>}]
 dyn_jacketStrap_01 [{shapeType=Capsule, length=7.0855, radius=2.1245, shapeName=dyn_jacketStrap_01_capsule, center=<4.5446 2 0>, rotation=<rot 90 180 179.9994>}]
 dyn_jacketStrap_02 [{shapeType=Capsule, length=4.9904, radius=2.4796, shapeName=dyn_jacketStrap_02_capsule, center=<4.0771 2 0>, rotation=<rot 90 180 179.9994>}]
 dyn_jacketStrap_03 [{
   shapeType=Capsule
   length=2.0304
   radius=2.4796
   shapeName=dyn_jacketStrap_03_capsule
   center=<1.9311 2 0>
   rotation=<rot 90 179.9998 179.9992>
  }]
 dyn_pantStrap_L_01 [{shapeType=Capsule, length=1.1432, radius=3.1068, shapeName=dyn_pantStrap_L_01_capsule, center=<2.4875 0 3>, rotation=<rot 90 -90 -90.0001>}]
 dyn_pantStrap_L_02 [{
   shapeType=Capsule
   length=5.0843
   radius=4.1286
   shapeName=dyn_pantStrap_L_02_capsule
   center=<4.8483 0 3>
   rotation=<rot 90 -179.9802 -179.9803>
  }]
 dyn_pantStrap_L_03 [{
   shapeType=Capsule
   length=9.2015
   radius=3.5732
   shapeName=dyn_pantStrap_L_03_capsule
   center=<6.0594 0 3>
   rotation=<rot 90 -172.4476 -172.4477>
  }]
 dyn_pantStrap_L_04 [{
   shapeType=Capsule
   length=10.3009
   radius=4.0675
   shapeName=dyn_pantStrap_L_04_capsule
   center=<7.4227 0 3>
   rotation=<rot 90 -176.6277 -176.6278>
  }]
 dyn_pantStrap_R_01 [{shapeType=Capsule, length=2.5, radius=2, shapeName=dyn_pantStrap_R_01_capsule, center=<2.6087 0 0>, rotation=<rot 90 180 179.9999>}]
 dyn_pantStrap_R_02 [{shapeType=Capsule, length=0.75, radius=1.5, shapeName=dyn_pantStrap_R_02_capsule, center=<6.72 0 -0.1968>, rotation=<rot 90 -90 -90.0001>}]
 dyn_pantStrap_R_03 [{shapeType=Capsule, length=5, radius=2, shapeName=dyn_pantStrap_R_03_capsule, center=<4.3108 0 0>, rotation=<rot 90 -166.7373 -166.7374>}]
 dyn_pantStrap_R_04 [{shapeType=Capsule, length=5, radius=2, shapeName=dyn_pantStrap_R_04_capsule, center=<4.8553 0 0>, rotation=<rot 90 -90 -90.0001>}]
 spine_02 [{shapeType=Capsule, length=14.7719, radius=11.9998, shapeName=spine_02_capsule}]
 spine_03 [
  {shapeType=Capsule, length=14.7719, radius=11.9998, shapeName=spine_03_capsule}
  {shapeType=Capsule, length=12.4771, radius=4.2944, shapeName=spine_03_capsule, center=<5.3309 10.4526 0.0925>, rotation=<rot 0 0 -13.0811>}
 ]
 thigh_l [{shapeType=Capsule, length=25.75, radius=9, shapeName=thigh_l_capsule, center=<-20.6567 0.0001 0.869>, rotation=<rot 85 -0.0002 0.0001>}]
 thigh_r [{shapeType=Capsule, length=25.5, radius=9.25, shapeName=thigh_r_capsule, center=<20.7351 0 -0.8851>, rotation=<rot -85 179.9999 179.9999>}]
}
constraints [
 {
  bone1Name=calf_r
  bone2Name=thigh_r
  swing1LimitDegrees=45
  swing1Motion=Free
  swing2LimitDegrees=45
  swing2Motion=Free
  twistLimitDegrees=45
  twistMotion=Free
 }
 {
  bone1Name=calf_l
  bone2Name=thigh_l
  swing1LimitDegrees=45
  swing1Motion=Free
  swing2LimitDegrees=45
  swing2Motion=Free
  twistLimitDegrees=45
  twistMotion=Free
 }
 {bone1Name=dyn_jacketStrap_01, bone2Name=thigh_l, swing1LimitDegrees=45, swing2LimitDegrees=45, twistLimitDegrees=45, twistMotion=Locked}
 {
  bone1Name=dyn_jacketStrap_02
  bone2Name=dyn_jacketStrap_01
  swing1LimitDegrees=45
  swing2LimitDegrees=45
  twistLimitDegrees=45
  twistMotion=Locked
 }
 {
  bone1Name=dyn_jacketStrap_03
  bone2Name=dyn_jacketStrap_02
  swing1LimitDegrees=45
  swing2LimitDegrees=45
  twistLimitDegrees=45
  twistMotion=Locked
 }
 {bone1Name=dyn_pantStrap_L_01, bone2Name=thigh_l, swing1LimitDegrees=45, swing2LimitDegrees=45, twistLimitDegrees=3}
 {bone1Name=dyn_pantStrap_L_02, bone2Name=dyn_pantStrap_L_01, swing1LimitDegrees=45, swing2LimitDegrees=45, twistLimitDegrees=3}
 {bone1Name=dyn_pantStrap_L_03, bone2Name=dyn_pantStrap_L_02, swing1LimitDegrees=45, swing2LimitDegrees=45, twistLimitDegrees=3}
 {bone1Name=dyn_pantStrap_L_04, bone2Name=dyn_pantStrap_L_03, swing1LimitDegrees=45, swing2LimitDegrees=45, twistLimitDegrees=3}
 {bone1Name=dyn_pantStrap_R_01, bone2Name=thigh_r, swing1LimitDegrees=45, swing2LimitDegrees=45, twistLimitDegrees=45, twistMotion=Locked}
 {
  bone1Name=dyn_pantStrap_R_02
  bone2Name=dyn_pantStrap_R_01
  swing1LimitDegrees=45
  swing2LimitDegrees=45
  twistLimitDegrees=45
  twistMotion=Locked
 }
 {
  bone1Name=dyn_pantStrap_R_03
  bone2Name=dyn_pantStrap_R_02
  swing1LimitDegrees=45
  swing2LimitDegrees=45
  twistLimitDegrees=45
  twistMotion=Locked
 }
 {
  bone1Name=dyn_pantStrap_R_04
  bone2Name=dyn_pantStrap_R_03
  swing1LimitDegrees=45
  swing2LimitDegrees=45
  twistLimitDegrees=45
  twistMotion=Locked
 }
 {
  bone1Name=spine_03
  bone2Name=spine_02
  swing1LimitDegrees=45
  swing1Motion=Free
  swing2LimitDegrees=45
  swing2Motion=Free
  twistLimitDegrees=45
  twistMotion=Free
 }
 {
  bone1Name=dyn_chestZipper_MSTR
  bone2Name=spine_03
  swing1LimitDegrees=45
  swing1Motion=Free
  swing2LimitDegrees=45
  swing2Motion=Free
  twistLimitDegrees=45
  twistMotion=Free
 }
 {bone1Name=dyn_chestZipper, bone2Name=dyn_chestZipper_MSTR, swing1LimitDegrees=45, swing2LimitDegrees=45, twistLimitDegrees=5}
]
