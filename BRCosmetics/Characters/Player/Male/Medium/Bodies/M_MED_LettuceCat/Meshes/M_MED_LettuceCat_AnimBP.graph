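# Post-process style cosmetic AnimBP for the LettuceCat medium male body, adding secondary motion to clothing bones on the pose copied from the attached parent mesh.
# Weighted multi-bone constraints place zipper, pant loop and strap bases, then RigidBody nodes simulate straps and loops.
# Spring-based AnimDynamics moves the hood, with additive bone offsets adjusting strap and hoodie.
# The DeformRig Control Rig node is unconnected; each update sets Gravity Override through a deprecated state function.
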AnimBlueprint /BRCosmetics/Characters/Player/Male/Medium/Bodies/M_MED_LettuceCat/Meshes/M_MED_LettuceCat_AnimBP
target_skeleton=&/M_MED_LettuceCat_Skeleton
graphs [
 AnimGraph {nodes [
   Control Rig() {
    alphaBoolBlend {blendOption=Linear}
    controlRigAssetReference {blueprintRigClass=&/M_MED_LettuceCat_DeformRig_ControlRig.M_MED_LettuceCat_DeformRig_ControlRig_C}
   }
   n1 = "Constraint - dyn_chestZipper_MSTR (PPPPPP)"!(
    ComponentPose=Local To Component(LocalPose=Copy Pose From Mesh() {bUseAttachedParent=true})
    Alpha=1
    ConstraintWeights_0=0
    ConstraintWeights_1:float=0.2396
    ConstraintWeights_2:float=0.0114
    ConstraintWeights_3:float=0.0311
    ConstraintWeights_4:float=0.3797
    ConstraintWeights_5:float=0.3381
   ) -> (Pose) {
    alphaBoolBlend {blendOption=Linear}
    boneToModify {boneName=dyn_chestZipper_MSTR}
    constraintSetup [
     {targetBone {boneName=upperarm_twist_01_r}}
     {targetBone {boneName=spine_02}}
     {targetBone {boneName=spine_01}}
     {targetBone {boneName=spine_05}}
     {targetBone {boneName=spine_03}}
     {targetBone {boneName=spine_04}}
    ]
    constraintWeights [1,1,1,1,1,1]
   }
   n2 = "Constraint - dyn_pantLoop_R_fr_MSTR (PPPPPP)"!(
    ComponentPose=@n1
    Alpha=1
    ConstraintWeights_0=0.1843
    ConstraintWeights_1:float=0.0311
    ConstraintWeights_2:float=0
    ConstraintWeights_3:float=0.0011
    ConstraintWeights_4:float=0.0001
    ConstraintWeights_5:float=0.7834
   ) -> (Pose) {
    alphaBoolBlend {blendOption=Linear}
    boneToModify {boneName=dyn_pantLoop_R_fr_MSTR}
    constraintSetup [
     {targetBone {boneName=thigh_r}}
     {targetBone {boneName=pelvis}}
     {targetBone {boneName=spine_02}}
     {targetBone {boneName=spine_01}}
     {targetBone {boneName=thigh_twist_01_l}}
     {targetBone {boneName=thigh_twist_01_r}}
    ]
    constraintWeights [1,1,1,1,1,1]
   }
   n3 = "Constraint - dyn_pantLoop_L_fr_MSTR (PPPPPP)"!(
    ComponentPose=@n2
    Alpha=1
    ConstraintWeights_0=0.7872
    ConstraintWeights_1:float=0.0013
    ConstraintWeights_2:float=0.0341
    ConstraintWeights_3:float=0.001
    ConstraintWeights_4:float=0.1737
    ConstraintWeights_5:float=0.0028
   ) -> (Pose) {
    alphaBoolBlend {blendOption=Linear}
    boneToModify {boneName=dyn_pantLoop_L_fr_MSTR}
    constraintSetup [
     {targetBone {boneName=thigh_l}}
     {targetBone {boneName=thigh_r}}
     {targetBone {boneName=pelvis}}
     {targetBone {boneName=spine_01}}
     {targetBone {boneName=thigh_twist_01_l}}
     {targetBone {boneName=thigh_twist_01_r}}
    ]
    constraintWeights [1,1,1,1,1,1]
   }
   n4 = "Constraint - dyn_PantLoop_L_MSTR (PPPPPP)"!(
    ComponentPose=@n3
    Alpha=1
    ConstraintWeights_0=0.1835
    ConstraintWeights_1:float=0.0041
    ConstraintWeights_2:float=0.2951
    ConstraintWeights_3:float=0.0032
    ConstraintWeights_4:float=0.4567
    ConstraintWeights_5:float=0.0574
   ) -> (Pose) {
    alphaBoolBlend {blendOption=Linear}
    boneToModify {boneName=dyn_PantLoop_L_MSTR}
    constraintSetup [
     {targetBone {boneName=thigh_l}}
     {targetBone {boneName=thigh_r}}
     {targetBone {boneName=pelvis}}
     {targetBone {boneName=spine_01}}
     {targetBone {boneName=thigh_twist_01_l}}
     {targetBone {boneName=thigh_twist_01_r}}
    ]
    constraintWeights [1,1,1,1,1,1]
   }
   n5 = "Constraint - dyn_pantLoop_R_MSTR (PPPPPP)"!(
    ComponentPose=@n4
    Alpha=1
    ConstraintWeights_0=0.0043
    ConstraintWeights_1:float=0.025
    ConstraintWeights_2:float=0.3875
    ConstraintWeights_3:float=0.0024
    ConstraintWeights_4:float=0.0267
    ConstraintWeights_5:float=0.554
   ) -> (Pose) {
    alphaBoolBlend {blendOption=Linear}
    boneToModify {boneName=dyn_pantLoop_R_MSTR}
    constraintSetup [
     {targetBone {boneName=thigh_l}}
     {targetBone {boneName=thigh_r}}
     {targetBone {boneName=pelvis}}
     {targetBone {boneName=spine_01}}
     {targetBone {boneName=thigh_twist_01_l}}
     {targetBone {boneName=thigh_twist_01_r}}
    ]
    constraintWeights [1,1,1,1,1,1]
   }
   n6 = "Constraint - dyn_pantStrap_L_base (PP)"!(ComponentPose=@n5, Alpha=1, ConstraintWeights_0=0.4363, ConstraintWeights_1:float=0.5637) -> (Pose) {
    alphaBoolBlend {blendOption=Linear}
    boneToModify {boneName=dyn_pantStrap_L_base}
    constraintSetup [{targetBone {boneName=thigh_r}}, {targetBone {boneName=thigh_twist_01_r}}]
    constraintWeights [1,1]
   }
   n7 = "Constraint - dyn_pantStrap_R_base (PPPPP)"!(
    ComponentPose=@n6
    Alpha=1
    ConstraintWeights_0=0.3303
    ConstraintWeights_1:float=0.0261
    ConstraintWeights_2:float=0.0249
    ConstraintWeights_3:float=0.618
    ConstraintWeights_4:float=0.0007
   ) -> (Pose) {
    alphaBoolBlend {blendOption=Linear}
    boneToModify {boneName=dyn_pantStrap_R_base}
    constraintSetup [
     {targetBone {boneName=thigh_l}}
     {targetBone {boneName=calf_twist_02_l}}
     {targetBone {boneName=calf_twist_01_l}}
     {targetBone {boneName=thigh_twist_01_l}}
     {targetBone {boneName=calf_l}}
    ]
    constraintWeights [1,1,1,1,1]
   }
   n8 = "Constraint - dyn_jacketStrap_base (PPPPPP)"!(
    ComponentPose=@n7
    Alpha=1
    ConstraintWeights_0=0.992
    ConstraintWeights_1:float=0.0002
    ConstraintWeights_2:float=0.0035
    ConstraintWeights_3:float=0.0001
    ConstraintWeights_4:float=0.0037
    ConstraintWeights_5:float=0.0005
   ) -> (Pose) {
    alphaBoolBlend {blendOption=Linear}
    boneToModify {boneName=dyn_jacketStrap_base}
    constraintSetup [
     {targetBone {boneName=thigh_l}}
     {targetBone {boneName=thigh_r}}
     {targetBone {boneName=pelvis}}
     {targetBone {boneName=spine_01}}
     {targetBone {boneName=thigh_twist_01_l}}
     {targetBone {boneName=thigh_twist_01_r}}
    ]
    constraintWeights [1,1,1,1,1,1]
   }
   n9 = RigidBody!(ComponentPose=@n8, bAlphaBoolEnabled:bool, OverrideWorldGravity:Vector="0,0,-980", ExternalForce) -> (Pose) {
    alphaInputType=Bool
    bClampLinearTranslationLimitToRefPose=true
    bOverrideWorldGravity=true
    overridePhysicsAsset=&/M_MED_LettuceCat_PantStrap_Physics
    simulationSpace=BaseBoneSpace
    alphaBoolBlend {blendOption=Linear}
    baseBoneRef {boneName=root}
    componentAppliedLinearAccClamp=<2500 2500 2500>
    componentLinearAccScale=<0.3 0.3 0.4>
    componentLinearVelScale=<0.3 0.3 0.4>
    overrideWorldGravity {z=-980}
    simSpaceSettings {
     maxAngularAcceleration=1000
     maxAngularVelocity=3
     maxLinearAcceleration=700
     maxLinearVelocity=500
     velocityScaleZ=0.5
     externalLinearDragV=<0.2 0.2>
    }
   }
   n10 = RigidBody(ComponentPose=@n9, OverrideWorldGravity="0,0,-980") {
    alphaInputType=Bool
    bClampLinearTranslationLimitToRefPose=true
    bOverrideWorldGravity=true
    overridePhysicsAsset=&/M_MED_LettuceCat_PantLoop_Physics
    simulationSpace=BaseBoneSpace
    alphaBoolBlend {blendOption=Linear}
    baseBoneRef {boneName=pelvis}
    componentAppliedLinearAccClamp=<2500 2500 2500>
    componentLinearAccScale=<0.3 0.3 0.3>
    componentLinearVelScale=<0.3 0.3 0.3>
    overrideWorldGravity {z=-980}
    simSpaceSettings {
     maxAngularAcceleration=1000
     maxAngularVelocity=3
     maxLinearAcceleration=700
     maxLinearVelocity=500
     velocityScaleZ=0.15
     worldAlpha=0.7
     externalLinearDragV=<0.2 0.2>
    }
   }
   n11 = "Transform (Modify) Bone - Bone: dyn_pantStrap_L_03"(ComponentPose=@n10, Alpha=1, Translation="-3.5,0,0.5", Rotation="0,0,0", Scale="1,1,1") {
    translationMode=BMM_Additive
    alphaBoolBlend {blendOption=Linear}
    alphaScaleBiasClamp {bInterpResult=true}
    boneToModify {boneName=dyn_pantStrap_L_03}
    translation {x=-3.5, z=0.5}
   }
   n12 = Anim Dynamics - Bone: dyn_hood_01!(ComponentPose=@n11, Alpha=0.7, GravityOverride:Vector="0,0,0", ExternalForce="0,0,0") -> (Pose) {
    angularDampingOverride=0.9999
    angularSpringConstant=75
    bAngularSpring=true
    bGravityOverrideInSimSpace=true
    bOverrideAngularDamping=true
    simulationSpace=RootRelative
    alphaBoolBlend {blendOption=Linear}
    boundBone {boneName=dyn_hood_01}
    componentAppliedLinearAccClamp=<5000 5000 5000>
    componentLinearAccScale=<1 1 1>
    componentLinearVelScale=<1 1 1>
    physicsBodyDefinitions [{
      sphereCollisionRadius=0
      boundBone {boneName=dyn_hood_01}
      boxExtents=<10 20 20>
      constraintSetup {angularLimitsMax=<0 0 0>, angularLimitsMin=<0 -70 0>, angularTarget=<0 0 0>}
      localJointOffset=<2 0 0>
     }]
   }
   n13 = Anim Dynamics - Bone: dyn_hood_02!(ComponentPose=@n12, Alpha=0.7, GravityOverride:Vector="0,0,0", ExternalForce="0,0,0") -> (Pose) {
    angularDampingOverride=0.9999
    angularSpringConstant=75
    bAngularSpring=true
    bGravityOverrideInSimSpace=true
    bOverrideAngularDamping=true
    simulationSpace=RootRelative
    alphaBoolBlend {blendOption=Linear}
    boundBone {boneName=dyn_hood_02}
    componentAppliedLinearAccClamp=<5000 5000 5000>
    componentLinearAccScale=<1 1 1>
    componentLinearVelScale=<1 1 1>
    physicsBodyDefinitions [{
      sphereCollisionRadius=0
      boundBone {boneName=dyn_hood_02}
      boxExtents=<10 20 20>
      constraintSetup {angularLimitsMax=<0 0 0>, angularLimitsMin=<0 -50 0>, angularTarget=<0 0 0>}
      localJointOffset=<2 0 0>
     }]
   }
   n14 = "Transform (Modify) Bone - Bone: dyn_hoodieTop"!(ComponentPose=@n13, bAlphaBoolEnabled:bool=True, Translation="-2.5,-1.468,0.353", Rotation="0,0,0", Scale="1,1,1") -> (Pose) {
    alphaInputType=Bool
    rotationMode=BMM_Additive
    rotationSpace=BCS_BoneSpace
    scaleMode=BMM_Additive
    scaleSpace=BCS_BoneSpace
    translationMode=BMM_Additive
    translationSpace=BCS_BoneSpace
    alphaBoolBlend {blendInTime=0.1, blendOption=Linear, blendOutTime=0.1}
    boneToModify {boneName=dyn_hoodieTop}
    translation=<-2.5 -1.468 0.353>
   }
   n15 = "Transform (Modify) Bone - Bone: dyn_hoodieTop"(ComponentPose=@n14, bAlphaBoolEnabled=True, Translation="4,1.468,-0.353", Rotation="0,0,0", Scale="1,1,1") {
    alphaInputType=Bool
    rotationMode=BMM_Additive
    rotationSpace=BCS_BoneSpace
    scaleMode=BMM_Additive
    scaleSpace=BCS_BoneSpace
    translationMode=BMM_Additive
    translationSpace=BCS_BoneSpace
    alphaBoolBlend {blendInTime=0.1, blendOption=Linear, blendOutTime=0.1}
    boneToModify {boneName=dyn_hoodieTop}
    translation=<4 1.468 -0.353>
   }
   n16 = "Transform (Modify) Bone - Bone: dyn_pantStrap_L_base"!(ComponentPose=@n15, bAlphaBoolEnabled:bool=True, Translation="0,0,0", Rotation="-0.015,3.922,0.057", Scale="1,1,1") -> (Pose) {
    alphaInputType=Bool
    rotationMode=BMM_Additive
    rotationSpace=BCS_BoneSpace
    scaleMode=BMM_Additive
    scaleSpace=BCS_BoneSpace
    translationMode=BMM_Additive
    translationSpace=BCS_BoneSpace
    alphaBoolBlend {blendInTime=0.1, blendOption=Linear, blendOutTime=0.1}
    boneToModify {boneName=dyn_pantStrap_L_base}
    rotation=<rot -0.015 3.922 0.057>
   }
   Output Pose(Result=Component To Local(ComponentPose=@n16)) {layerGroup=None}
  ]}
 EventGraph {nodes [
   Event BlueprintUpdateAnimation()
   > n1 = Gravity Override State Function DEPRECATED(CharacterPartAnimInstance=Self-Reference(), GameTimeInSeconds=GetGameTimeInSeconds(), GravityOverrideParams=Get Gravity Override Params())
   > Set Gravity Override(Gravity Override=@n1)
  ]}
]
variables [
 Gravity Override {type=Vector}
 Gravity Override Params {type=GravityOverrideParamsStruct}
]
